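# Material for the prototype Doggus sword on skeletal meshes, adding tunable glow and an animated Zero Point charge effect.
# Weapon master output gets scaled emissive and an FX-mask-driven base color override, then passes through base customization.
# MF_ZeroPointCharge then applies energy direction, mask, brightness and progress from the shared MPC_Doggus collection, before MF_CreativeMoveTool.
# A separate roughness-offset branch reads the weapon master output but does not feed that chain.
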Material /Game/Athena/Prototype/Blueprints/Doggus_Sword/Materials/M_Doggus_Sword_Anim
graph {
 properties {
  bUseMaterialAttributes=true
  bUsedWithMeshDeformer=true
  bUsedWithSkeletalMesh=true
  pixelDepthOffsetMode=PDOM_Legacy
  refractionMethod=RM_IndexOfRefraction
  usedShadingModels=Default Lit
 }
 nodes [
  n1 = MaterialFunctionCall!(In) {materialFunction=/Game/Packages/Fortress_SharedMaterials/Base_Material_Functions/BaseCustomizationFunctions/MF_FN_Weapon_MASTER}
  n2 = GetMaterialAttributes(None=@n1.Material Result) {attributeGetTypes [#1]}
  n3 = ScalarParameter() {defaultValue=1, parameterName=DoggusSwordGlow}
  n4 = TextureSample() {
   samplerSource=SSM_Wrap_WorldGroupSettings
   samplerType=SAMPLERTYPE_Masks
   texture=/Game/Athena/Prototype/Blueprints/Doggus_Sword/Textures/T_Doggus_Sword_FX
  }
  n5 = MaterialFunctionCall!(
   Material=SetMaterialAttributes(MaterialAttributes=@n2.MaterialAttributes, Emissive Color=Multiply(A=@n3, B=@n2.Emissive Color)) {attributeSetTypes [#1]}
   NewBaseColor=Multiply(A=Multiply(A=@n2.Emissive Color, B=@n4.R), B=ScalarParameter() {parameterName=DoggusSwordMaskAmount})
   Mask=Multiply(A=@n4.R, B=ScalarParameter() {parameterName=DoggusSwordMask2})
  ) {materialFunction=/Engine/Functions/MaterialLayerFunctions/MatLayerBlend_OverrideBaseColor}
  n6 = VectorParameter() {parameterName=WeaponForwardAxis, defaultValue=<rgba 0 1 0 1>}
  n7 = Multiply(A=ScalarParameter() {defaultValue=1, parameterName=DoggusSwordMask1}, B=@n4.R)
  n8 = CollectionParameter() {collection=/Game/Characters/Player/Male/Medium/Bodies/M_MED_Doggus/MaterialSet/MPC_Doggus, parameterName=Energy_SubtractMaskDistance_Sword}
  MaterialFunctionCall(MA In=MaterialFunctionCall!(
    In=MaterialFunctionCall!(
     Material=@n5.Blended Material
     ------------------
     Object Bounds Center Adjustment
     Object Bounds Min Adjustment
     Wrap ScaleMultiplier
     Rotation X Axis
     "Customization Mask(_CM A)"
     Forward Axis=@n6.RGB
     Object Bounds Max Adjustment
     Rotation Z Axis
     Normal Map=@n1.Normal Map
     Cropped Emissive Mask=@n1.Normal Map
     "Customization Crevasse (_CM G)"=@n1.Normal Map
     "Customization Mask UV Islands (_CM B Bottom Right)"=@n1.Normal Map
     -------------------
     "Customization Mask 1 (_CM B Top Left)"=@n7
     "Customization Mask 2 (_CM B Top Right)"=@n7
     "Customization Mask 3 (_CM B Bottom Left)"=@n7
     "Customization Mask (_CM RGBA)"
     --------------------
     Pre-Skinned Local Position=@n1.Normal Map
     Pre-Skinned Bounds=@n1.Normal Map
     Pre-Skinned Local Normal=@n1.Normal Map
     Per_ObjectMaskOverride
     isVehicle
     Side Project Vehicle
     isMech
     isCharacter
     isCharAccent
     isPickaxe
     isBackpack
     Object-SafeZone-Back
     Object-SafeZone-Middle
     Object-SafeZone-Front
     Object-SafeZone-VP_Charcter_Preview_Switch
     Object-SafeZone-Virtual-Pivot
     Wrap Holdout
    ) {materialFunction=/Game/Packages/Fortress_SharedMaterials/Base_Material_Functions/BaseCustomizationFunctions/MF_Base_Customization}.Result
    Mask
    EnergyDirection=CollectionParameter() {collection=&MPC_Doggus, parameterName=Energy_Direction_Sword}
    EnergySubtractMask=@n8
    EnergyBrightness=CollectionParameter() {collection=&MPC_Doggus, parameterName=Engergy_Brightness_Sword}
    EnergyProgress=CollectionParameter() {collection=&MPC_Doggus, parameterName=Energy_Progress_Sword}
   ) {materialFunction=/Game/Characters/Player/Male/Medium/Bodies/M_MED_Doggus/MaterialSet/MF_ZeroPointCharge}.Result) {materialFunction=&MF_CreativeMoveTool}
  n9 = GetMaterialAttributes(None=@n1.Material Result) {attributeGetTypes [#2]}
  n10 = ScalarParameter() {parameterName=DoggusSwordRoughAdd}
  SetMaterialAttributes!(MaterialAttributes=@n9.MaterialAttributes, Roughness=Saturate(None=Add(A=@n9.Roughness, B=@n10))) {attributeSetTypes [#2]}
 ]
}
